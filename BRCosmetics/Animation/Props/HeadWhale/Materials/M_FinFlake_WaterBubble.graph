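# Translucent lit glass material for the FinFlake water bubble on the HeadWhale prop.
# MF_Glass feeds a thin-film function, with static switches for camera dither and for vertex hide versus character opacity.
# An animated water normal, triplanar-projected on pre-skinned local position so it follows deformation, blends into the glass normal.
# Refraction and center roughness are then adjusted before TechArt overrides, and a custom HLSL node computes a scene-depth mask.
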Material /BRCosmetics/Animation/Props/HeadWhale/Materials/M_FinFlake_WaterBubble
graph {
 properties {
  bEnableResponsiveAA=true
  bScreenSpaceReflections=true
  bUseMaterialAttributes=true
  bUsedWithClothing=true
  bUsedWithMeshDeformer=true
  bUsedWithMorphTargets=true
  bUsedWithNanite=true
  bUsedWithNiagaraMeshParticles=true
  bUsedWithSkeletalMesh=true
  blendMode=BLEND_Translucent
  ditherOpacityMask=true
  pixelDepthOffsetMode=PDOM_Legacy
  refractionMethod=RM_PixelNormalOffset
  translucencyDirectionalLightingIntensity=5
  translucencyLightingMode=TLM_SurfacePerPixelLighting
  usedShadingModels=Default Lit
 }
 nodes [
  n1 = TextureSampleParameter2D() {
   group=00_Glass
   parameterName=TechArtMask
   samplerType=SAMPLERTYPE_Masks
   sortPriority=-96
   texture=&T_BlackMask
  }
  n2 = MaterialFunctionCall!(
   Material Attributes=MaterialFunctionCall!(TechArtMask=@n1.RGB) {materialFunction=/Game/Global/Materials/Character_Parent_2023/Glass_Parent_2023/MF_Glass}.Result
   Fresnel=Fresnel() {baseReflectFraction=0, exponent=1}
   Alpha=@n1.RGB
  ) {materialFunction=/Game/Global/MaterialFunctions/TechArt/MF_ThinFilm_TechArt}
  n3 = StaticSwitchParameter(
   True=MaterialFunctionCall!(In=@n2.Result) {materialFunction=/Game/Packages/Fortress_SharedMaterials/Base_Material_Functions/BaseCustomizationFunctions/MF_FN_Vehicle_CameraDither}.Result
   False=@n2.Result
  ) {group=00_Glass, parameterName=UseCameraDither, sortPriority=250}
  n4 = Multiply(A=Subtract(A=ObjectPositionWS(), B=WorldPosition().XYZ), B=ScalarParameter() {group=00_Glass, parameterName=VertexHide, sortPriority=301})
  n5 = GetMaterialAttributes(None=StaticSwitchParameter(
    True=SetMaterialAttributes!(MaterialAttributes=@n3, World Position Offset=@n4) {attributeSetTypes [#1]}
    False=MaterialFunctionCall!(Material=@n3) {materialFunction=/Game/Global/Materials/Character_Parent_2023/Functions/BaseFN_CharacterOpacity_2023}.Result
   ) {group=00_Glass, parameterName=UseVertexHide, sortPriority=300}) {attributeGetTypes [#2, #3]}
  n6 = ScalarParameter() {defaultValue=1.2, group=FinFlake WaterBubble, parameterName=WaterBubble_Refraction}
  n7 = NamedRerouteUsage()
  n8 = ScalarParameter() {defaultValue=1, group=FinFlake WaterBubble, parameterName=CenterRoughness}
  n9 = NamedRerouteUsage()
  n10 = Saturate(None=Power(Base=OneMinus(None=@n9)) {constExponent=4})
  MaterialFunctionCall(In=SetMaterialAttributes!(
    MaterialAttributes=@n5.MaterialAttributes
    Refraction=LinearInterpolate(B=@n6, Alpha=@n7) {constA=1}
    Normal=NamedRerouteUsage()
    Roughness=LinearInterpolate(A=@n5.Roughness, B=@n8, Alpha=@n10)
   ) {attributeSetTypes [#4, #2, #3]}) {materialFunction=&MF_TechArtOverrides}
  n11 = NamedRerouteUsage()
  NamedRerouteDeclaration(None=Fresnel(Normal=Transform(None=@n11)) {baseReflectFraction=0, exponent=1}) {name=Fresnel, nodeColor=<rgba 0 0.2031 1 1>}
  n12 = NamedRerouteUsage()
  n13 = Multiply(A=Abs(None=ScalarParameter() {defaultValue=200, group=FinFlake WaterBubble, parameterName=WaterNormalScale})) {constB=-1}
  n14 = LocalPosition() {localOrigin=InstancePreSkinning}
  n15 = Add(
   A=NamedRerouteDeclaration(None=VertexInterpolator(VS=@n14.XYZ).PS) {name=PS-LocPos, nodeColor=<rgba 1 0 0.7031 1>}
   B=Multiply(A=Time(), B=ScalarParameter() {defaultValue=4, group=FinFlake WaterBubble, parameterName=WaterNormalPanSpeed})
  )
  n16 = TextureObject() {samplerType=SAMPLERTYPE_Normal, texture=/Game/Athena/Apollo/Water/Textures/02_Rounded/02_WaterFoam_Veroni_base_N}
  n17 = VertexNormalWS()
  n18 = ComponentMask(None=@n17) {b=true}
  n19 = Constant() {r=1}
  n20 = ComponentMask(None=MaterialFunctionCall(In=Abs(None=@n18), Contrast=@n19) {materialFunction=&CheapContrast}.Result) {desc="1=+z", r=true}
  n21 = ComponentMask(None=@n17) {r=true}
  n22 = ComponentMask(None=MaterialFunctionCall(In=Abs(None=@n21), Contrast=@n19) {materialFunction=&CheapContrast}.Result) {desc="1=+x", r=true}
  n23 = MaterialFunctionCall(
   Normal=Transform(None=MaterialFunctionCall!(Negative World Position Divisor=@n13, World Position=@n15, Texture=@n16, Z lerp Alpha=@n20, WorldNormal=@n17, X lerp alpha=@n22) {materialFunction=/Engine/Functions/Engine_MaterialFunctions02/WorldAlignedNormals_HighQuality}.World Space Normals XYZ) {transformSourceType=TRANSFORMSOURCE_World, transformType=TRANSFORM_Tangent}
   Flatness=ScalarParameter() {defaultValue=0.5, group=FinFlake WaterBubble, parameterName=WaterNormal_Flatness}
  ) {materialFunction=&FlattenNormal}
  NamedRerouteDeclaration(None=LinearInterpolate(
    A=@n12
    B=MaterialFunctionCall(BaseNormal=@n12, AdditionalNormal=@n23.Result) {materialFunction=&BlendAngleCorrectedNormals}.Result
    Alpha=NamedRerouteUsage()
   )) {name=Normal_OUT, nodeColor=<rgba 0.0078 0 1 1>}
  n24 = SceneDepth()
  NamedRerouteDeclaration(None=Multiply(A=Saturate(None=@n24.R), B=PixelDepth().R)) {name=PDepth, nodeColor=<rgba 0 1 0.8594 1>}
  NamedRerouteDeclaration(None=@n5.Normal) {name=Normal_IN, nodeColor=<rgba 0 1 0.9063 1>}
  NamedRerouteDeclaration(None=Constant()) {name=PixelDepthOffset, nodeColor=<rgba 0.2188 0 1 1>}
  NamedRerouteDeclaration(None=ScalarParameter() {defaultValue=50, group=SoftGlow, parameterName=MaxBlurDepth, sortPriority=22}) {name=MaxBlurDepth, nodeColor=<rgba 0.1016 1 0 1>}
  NamedRerouteDeclaration(None=ScalarParameter() {defaultValue=1, parameterName=DistanceMask}) {name=DistanceMask, nodeColor=<rgba 0 1 0.8594 1>}
  MaterialExpressionNamedRerouteUsage_31 = NamedRerouteUsage()
  MaterialExpressionNamedRerouteUsage_50 = NamedRerouteUsage()
  MaterialExpressionNamedRerouteUsage_29 = NamedRerouteUsage()
  MaterialExpressionNamedRerouteUsage_28 = NamedRerouteUsage()
  n25 = Custom!(
   DistanceMask=@MaterialExpressionNamedRerouteUsage_31
   DivideF=@MaterialExpressionNamedRerouteUsage_50
   PDepth=@MaterialExpressionNamedRerouteUsage_29
   PixelDepthOffset=@MaterialExpressionNamedRerouteUsage_28
  ) {
   code="float2 baseUV = MaterialFloat2(ScreenAlignedPosition(Parameters.ScreenPosition).xy);\r\nfloat tmpdepth2 = CalcSceneDepth(baseUV)-PDepth+PixelDepthOffset;\r\nfloat mask = (min(1-saturate(max(0,(tmpdepth2))/(DivideF+PixelDepthOffset))-0.99,0)+0.99)*saturate(sign(tmpdepth2));\r\nreturn mask;"
   desc="\n*** Original source before expression upgrade ***\nfloat3 CurColor;\r\nfloat2 BaseUV = MaterialFloat2(ScreenAlignedPosition(Parameters.ScreenPosition).xy);\r\nfloat2 NewUV = BaseUV;\r\nfloat StepSize = Distance / (int) DistanceSteps;\r\nfloat CurDistance=0;\r\nfloat2 CurOffset;\r\nfloat TwoPi = 6.283185;\r\nfloat Substep;\r\nfloat2 ScenePixels=View.RenderTargetSize*BaseUV;\r\nScenePixels+=View.TemporalAAParams.r;\r\nfloat2 RandomSamp = ((uint)(ScenePixels.x) + 2 * (uint)(ScenePixels.y)) % 5;\r\nRandomSamp+=Texture2DSample(Tex,TexSampler,ScenePixels);\r\nRandomSamp/=5;\r\nRandomSamp-=0.5;\r\nTempAARotation*=RandomSamp;\r\nTempAADistance*=StepSize*RandomSamp;\r\n\r\n\r\nint i=0;\r\nif (DistanceSteps<1)\r\n{\r\nreturn DecodeSceneColorForMaterialNode(NewUV);\r\n}\r\nelse\r\n{\r\n//CurDistance += 0.5*StepSize;\r\nwhile ( i < (int) DistanceSteps)\r\n{\r\n\r\n//CurDistance+=StepSize;\r\nfor (int j = 0; j < (int) RadialSteps; j++)\r\n{\r\nCurOffset.x = cos(TwoPi*((TempAARotation+Substep) / RadialSteps));\r\nCurOffset.y = sin(TwoPi*((TempAARotation+Substep) / RadialSteps));\r\nCurOffset *=DistanceMask;\r\nNewUV.x = BaseUV.x + (CurOffset.x * (CurDistance+(RandomSamp*TempAADistance)));\r\nNewUV.y = BaseUV.y + (CurOffset.y * (CurDistance+(RandomSamp*TempAADistance)));\r\nCurColor += DecodeSceneColorForMaterialNode(NewUV);\r\n//CurDistance+=(StepSize+(TempAADistance))/RadialSteps;\r\nSubstep++;\r\n}\r\nCurDistance+=StepSize;\r\nSubstep+=RadialOffset;\r\ni++;\r\n}\r\nCurColor = CurColor / ((int)DistanceSteps*(int)RadialSteps);\r\nreturn CurColor;\r\n}\n*** Original source before expression upgrade ***\nfloat3 CurColor;\r\nfloat2 BaseUV = MaterialFloat2(ScreenAlignedPosition(Parameters.ScreenPosition).xy);\r\nfloat2 NewUV = BaseUV;\r\nfloat StepSize = Distance / (int) DistanceSteps;\r\nfloat CurDistance=0;\r\nfloat2 CurOffset;\r\nfloat TwoPi = 6.283185;\r\nfloat Substep;\r\nfloat2 ScenePixels=Frame.RenderTargetSize*BaseUV;\r\nScenePixels+=Frame.TemporalAAParams.r;\r\nfloat2 RandomSamp = ((uint)(ScenePixels.x) + 2 * (uint)(ScenePixels.y)) % 5;\r\nRandomSamp+=Texture2DSample(Tex,TexSampler,ScenePixels);\r\nRandomSamp/=5;\r\nRandomSamp-=0.5;\r\nTempAARotation*=RandomSamp;\r\nTempAADistance*=StepSize*RandomSamp;\r\n\r\n\r\nint i=0;\r\nif (DistanceSteps<1)\r\n{\r\nreturn DecodeSceneColorForMaterialNode(NewUV);\r\n}\r\nelse\r\n{\r\n//CurDistance += 0.5*StepSize;\r\nwhile ( i < (int) DistanceSteps)\r\n{\r\n\r\n//CurDistance+=StepSize;\r\nfor (int j = 0; j < (int) RadialSteps; j++)\r\n{\r\nCurOffset.x = cos(TwoPi*((TempAARotation+Substep) / RadialSteps));\r\nCurOffset.y = sin(TwoPi*((TempAARotation+Substep) / RadialSteps));\r\nCurOffset *=DistanceMask;\r\nNewUV.x = BaseUV.x + (CurOffset.x * (CurDistance+(RandomSamp*TempAADistance)));\r\nNewUV.y = BaseUV.y + (CurOffset.y * (CurDistance+(RandomSamp*TempAADistance)));\r\nCurColor += DecodeSceneColorForMaterialNode(NewUV);\r\n//CurDistance+=(StepSize+(TempAADistance))/RadialSteps;\r\nSubstep++;\r\n}\r\nCurDistance+=StepSize;\r\nSubstep+=RadialOffset;\r\ni++;\r\n}\r\nCurColor = CurColor / ((int)DistanceSteps*(int)RadialSteps);\r\nreturn CurColor;\r\n}"
   outputType=CMOT_Float1
   showCode=true
   inputs [
    {
     inputName=DistanceMask
     input {expression=&self:MaterialExpressionNamedRerouteUsage_31}
    }
    {
     inputName=DivideF
     input {expression=&self:MaterialExpressionNamedRerouteUsage_50}
    }
    {
     inputName=PDepth
     input {expression=&self:MaterialExpressionNamedRerouteUsage_29}
    }
    {
     inputName=PixelDepthOffset
     input {expression=&self:MaterialExpressionNamedRerouteUsage_28}
    }
   ]
  }
  NamedRerouteDeclaration(None=@n25) {name=DepthMask, nodeColor=<rgba 0.4062 0 1 1>}
 ]
}
